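# Physics asset for the M_MED_Hairy backpack cosmetic, giving the pack's dynamic hinge, two handles and fin bones bodies for secondary motion.
# A small box on the pack bone is the hub, with each dyn_ bone constrained directly to it rather than chained.
# The hinge and handles use boxes and the fin a capsule.
# Swing 1 is locked on the left handle and fin, and twist on the left handle only.
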PhysicsAsset /BRCosmetics/Accessories/FORT_Backpacks/Backpack_M_MED_Hairy/Meshes/M_MED_Hairy_Pack_Physics
bodies {
 dyn_fin_01 [{shapeType=Capsule, length=9.749, radius=1.8471, center=<3.9013 0 -2.7666>, rotation=<rot 0 -89.9998 125.3418>}]
 dyn_handle_l [{shapeType=Box, extentX=5.8336, extentY=1.857, extentZ=8.619, center=<1.9042 -0.0706 -0.095>, rotation=<rot 0 180 -90>}]
 dyn_handle_r [{shapeType=Box, extentX=5.8348, extentY=1.647, extentZ=8.619, center=<2.0439 0.0709 -0.1736>, rotation=<rot 0 0 -90>}]
 dyn_hinge [{shapeType=Box, extentX=17.881, extentY=21.6814, extentZ=5.2443, center=<8.8777 0 0>}]
 pack [{shapeType=Box, extentX=1.5, extentY=1, extentZ=1}]
}
constraints [
 {bone1Name=dyn_hinge, bone2Name=pack, swing1LimitDegrees=3, swing2LimitDegrees=30, twistLimitDegrees=3}
 {
  bone1Name=dyn_handle_l
  bone2Name=pack
  swing1LimitDegrees=0.1
  swing1Motion=Locked
  swing2LimitDegrees=40
  twistLimitDegrees=0.1
  twistMotion=Locked
 }
 {bone1Name=dyn_handle_r, bone2Name=pack, swing1LimitDegrees=0.1, swing2LimitDegrees=50, twistLimitDegrees=0.1}
 {bone1Name=dyn_fin_01, bone2Name=pack, swing1LimitDegrees=55, swing1Motion=Locked, swing2LimitDegrees=55, twistLimitDegrees=5}
]
